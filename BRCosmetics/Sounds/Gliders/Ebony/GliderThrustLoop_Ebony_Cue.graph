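# Looping thrust sound cue for the Ebony glider, layering static, tonal, wind, flame and whistle waves through a mixer.
# Each layer has its own continuous modulator that varies volume and pitch with movement parameters.
# Static, flame and whistle follow the Forwards parameter, while tonal and wind follow AnyDirection with non-zero volume floors.
# Glider-specific attenuation and a glider loop audio properties sheet configure playback.
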SoundCue /BRCosmetics/Sounds/Gliders/Ebony/GliderThrustLoop_Ebony_Cue
root=Mixer_1
nodes [
 {
  id=Mixer_1
  children [ModulatorContinuous_0, ModulatorContinuous_1, ModulatorContinuous_2, ModulatorContinuous_3, ModulatorContinuous_4]
  properties {inputVolume [0.8,1,1,0.75,1]}
 }
 {
  id=ModulatorContinuous_0
  children [WavePlayer_0]
  properties {
   pitchModulationParams {default=0, maxOutput=1.2, minOutput=1, parameterName=Forwards}
   volumeModulationParams {default=0, minOutput=0.1, parameterName=Forwards}
  }
 }
 {
  id=ModulatorContinuous_1
  children [WavePlayer_1]
  properties {
   pitchModulationParams {default=0, maxOutput=1.1, minOutput=1, parameterName=AnyDirection}
   volumeModulationParams {default=0, minOutput=0.2, parameterName=AnyDirection}
  }
 }
 {
  id=ModulatorContinuous_2
  children [WavePlayer_2]
  properties {
   pitchModulationParams {default=0, maxOutput=1.1, minOutput=1, parameterName=AnyDirection}
   volumeModulationParams {default=0, maxOutput=0.8, minOutput=0.4, parameterName=AnyDirection}
  }
 }
 {
  id=ModulatorContinuous_3
  children [WavePlayer_3]
  properties {pitchModulationParams {default=0, minOutput=0.99, parameterName=Forwards}, volumeModulationParams {default=0, parameterName=Forwards}}
 }
 {
  id=ModulatorContinuous_4
  children [WavePlayer_4]
  properties {
   pitchModulationParams {default=0, minOutput=0.99, parameterName=Forwards}
   volumeModulationParams {default=0, maxOutput=0.9, parameterName=Forwards}
  }
 }
 {id=WavePlayer_0, properties {bLooping=true, soundWaveAssetPtr=&/Glider_Ebony_Thurst_LP_Static}}
 {id=WavePlayer_1, properties {bLooping=true, soundWaveAssetPtr=&/Glider_Ebony_Thurst_LP_Tonal}}
 {id=WavePlayer_2, properties {bLooping=true, soundWaveAssetPtr=&/Glider_Ebony_Thurst_LP_Wind}}
 {id=WavePlayer_3, properties {bLooping=true, soundWaveAssetPtr=&/Glider_Ebony_Thurst_LP_Flame}}
 {id=WavePlayer_4, properties {bLooping=true, soundWaveAssetPtr=&/Glider_Ebony_Thurst_LP_Whistle}}
]
properties {
 attenuationSettings=&Fort_Player_Glider_Thrust_Athena_Att
 audioPropertiesSheet=&APS_Cosmetic_Glider_Loop
 maxDistance=12500
 soundClassObject=&Fort_Gliders_Land
 volumeMultiplier=1
 concurrencySet [null]
}
